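# Animation Blueprint for the GeneAglet backpack cosmetic, adjusting the pack bone and four WingFoil bones over the input pose.
# AnimGraph chains three bool-alpha additive pack offsets, then mirrored additive rolls on the upper and lower foils.
# Each update derives a standing flag and a smoothed Speed Alpha from velocity, targeting and skydiving state.
# Customization load sets Is LargeBody from the part's body type; no variable visibly feeds the AnimGraph.
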AnimBlueprint /BRCosmetics/Accessories/FORT_Backpacks/Backpack_M_MED_GeneAglet/Meshes/M_MED_GeneAglet_Pack_AnimBP
target_skeleton=&/M_MED_GeneAglet_Pack_Skeleton
graphs [
 AnimGraph {nodes [
   n1 = "Transform (Modify) Bone - Bone: pack"!(ComponentPose=Local To Component(LocalPose=Input Pose()), bAlphaBoolEnabled:bool, Translation="0,-7,0", Rotation="0,0,0", Scale) -> (Pose) {
    alphaInputType=Bool
    rotationMode=BMM_Additive
    rotationSpace=BCS_BoneSpace
    translationMode=BMM_Additive
    translationSpace=BCS_BoneSpace
    alphaBoolBlend {blendInTime=0.8, blendOutTime=0.8}
    boneToModify {boneName=pack}
    translation {y=-7}
   }
   n2 = "Transform (Modify) Bone - Bone: pack"(ComponentPose=@n1, Translation="0,-3,0", Rotation="0,0,0") {
    alphaInputType=Bool
    rotationMode=BMM_Additive
    rotationSpace=BCS_BoneSpace
    translationMode=BMM_Additive
    translationSpace=BCS_BoneSpace
    alphaBoolBlend {blendInTime=0.8, blendOutTime=0.8}
    boneToModify {boneName=pack}
    translation {y=-3}
   }
   n3 = "Transform (Modify) Bone - Bone: pack"(ComponentPose=@n2, Translation="0,5,0", Rotation="0,0,0") {
    alphaInputType=Bool
    rotationMode=BMM_Additive
    rotationSpace=BCS_BoneSpace
    translationMode=BMM_Additive
    translationSpace=BCS_BoneSpace
    alphaBoolBlend {blendInTime=0.8, blendOutTime=0.8}
    boneToModify {boneName=pack}
    translation {y=5}
   }
   n4 = "Transform (Modify) Bone - Bone: deform_WingFoil_r_up"(ComponentPose=@n3, Alpha=1, Translation="0,0,0", Rotation="0,0,70", Scale="1,1,1") {
    rotationMode=BMM_Additive
    rotationSpace=BCS_BoneSpace
    translationMode=BMM_Additive
    translationSpace=BCS_BoneSpace
    alphaBoolBlend {blendInTime=1, blendOption=Linear, blendOutTime=1}
    alphaScaleBiasClamp {bInterpResult=true}
    boneToModify {boneName=deform_WingFoil_r_up}
    rotation {roll=70}
   }
   n5 = "Transform (Modify) Bone - Bone: deform_WingFoil_r_down"(ComponentPose=@n4, Alpha=1, Translation="0,0,0", Rotation="0,0,-85", Scale="1,1,1") {
    rotationMode=BMM_Additive
    rotationSpace=BCS_BoneSpace
    translationMode=BMM_Additive
    translationSpace=BCS_BoneSpace
    alphaBoolBlend {blendInTime=1, blendOption=Linear, blendOutTime=1}
    alphaScaleBiasClamp {bInterpResult=true}
    boneToModify {boneName=deform_WingFoil_r_down}
    rotation {roll=-85}
   }
   n6 = "Transform (Modify) Bone - Bone: deform_WingFoil_l_up"(ComponentPose=@n5, Alpha=1, Translation="0,0,0", Rotation="0,0,-70", Scale="1,1,1") {
    rotationMode=BMM_Additive
    rotationSpace=BCS_BoneSpace
    translationMode=BMM_Additive
    translationSpace=BCS_BoneSpace
    alphaBoolBlend {blendInTime=1, blendOption=Linear, blendOutTime=1}
    alphaScaleBiasClamp {bInterpResult=true}
    boneToModify {boneName=deform_WingFoil_l_up}
    rotation {roll=-70}
   }
   n7 = "Transform (Modify) Bone - Bone: deform_WingFoil_l_down"(ComponentPose=@n6, Alpha=1, Translation="0,0,0", Rotation="0,0,85", Scale="1,1,1") {
    rotationMode=BMM_Additive
    rotationSpace=BCS_BoneSpace
    translationMode=BMM_Additive
    translationSpace=BCS_BoneSpace
    alphaBoolBlend {blendInTime=1, blendOption=Linear, blendOutTime=1}
    alphaScaleBiasClamp {bInterpResult=true}
    boneToModify {boneName=deform_WingFoil_l_down}
    rotation {roll=85}
   }
   Output Pose(Result=Component To Local(ComponentPose=@n7)) {layerGroup=None}
  ]}
 EventGraph {nodes [
   n1 = Event BlueprintUpdateAnimation()
   > Set Delta Time X(Delta Time X=@n1.DeltaTimeX)
   > n2 = Sequence()
   n3 = TryGetPawnOwner()
   n4 = GetVelocity!(self:Actor Object Reference=@n3) -> (ReturnValue:Vector)
   n5 = GetActorForwardVector!(self:Actor Object Reference=@n3) -> (ReturnValue:Vector)
   n6 = GetActorRightVector!(self:Actor Object Reference=@n3) -> (ReturnValue:Vector)
   n7 = float + float(A="Absolute (Float)"(A=Dot Product!(A:Vector=@n4, B:Vector=@n5) -> (ReturnValue:double)), B="Absolute (Float)"(A=Dot Product(A=@n4, B=@n6)))
   > n8 = Branch(Condition="Equal (Float)"(A=@n7))
   > Set is standing(is standing:bool=true) -> (Output_Get:bool)
   Set is standing(execute=@n8.else, is standing=false)
   n9 = AND Boolean(A=Get bIsTargeting(), B=Get bIsMoving2D())
   n10 = float + float(A=MapRangeClamped(Value=@n7, InRangeA=0, InRangeB=550), B:bool=Get bIsSkyDiving())
   Set Speed Alpha(
    execute=@n2.then_1
    Speed Alpha:double=FInterpTo(
     Current=Get Speed Alpha() -> (Speed Alpha:double)
     Target=SelectFloat(A=0.4, B=@n10, bPickA=@n9)
     DeltaTime=Get Delta Time X()
     InterpSpeed=6
    )
   ) -> (Output_Get:double)
   n11 = Event OnCharacterCustomizationLoaded()
   n12 = Get Array Item with EFort Custom Part Type(Array:Array of Custom Character Part Object References=@n11.CustomParts) -> (Output:Custom Character Part Object Reference)
   n13 = Get BodyTypesPermitted!(self:Custom Character Part Object Reference=@n12) -> (BodyTypesPermitted:EFortCustomBodyType Enum=NONE)
   > Set Is LargeBody(Is LargeBody:bool="Equal (Enum)"(A:EFortCustomBodyType Enum=@n13, B:EFortCustomBodyType Enum=Large)) -> (Output_Get:bool)
   n14 = Get GenderPermitted!(self:Custom Character Part Object Reference=@n12) -> (GenderPermitted:EFortCustomGender Enum=Invalid)
   "Equal (Enum)"(A:EFortCustomGender Enum=@n14, B:EFortCustomGender Enum=Male)
   "Equal (Enum)"(A:EFortCustomBodyType Enum=@n13, B:EFortCustomBodyType Enum=Medium)
   "Equal (Enum)"(A:EFortCustomGender Enum=@n14, B:EFortCustomGender Enum=Female)
  ]}
]
variables [
 Delay {type=number, default=2}
 Delay_timer {type=number, default=2}
 Delta Time X {type=number}
 Is LargeBody {type=boolean}
 is standing {type=boolean}
 Speed Alpha {type=number}
]
